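; NPC music component for the FNM TV Scare player pawn that switches music according to the pawn's alert level.
; On BeginPlay it casts its owner to BP_PlayerPawn_TVScare and binds to OnAlertLevelChanged, storing each new level in AlertLevel.
; AlertLevel's OnRep runs AttemptToChangeMusic only without authority.
; After a 0.2-second delay, that fires the Unaware or Aggro trigger on the music manager for Unaware or Threatened.
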(event EventBeginPlay
  (|CallParentFunction)
  (bind _asbp_player_pawn_tvscare (Utilities|Casting|CastToBP_PlayerPawn_TVScare (Components|GetOwner)))
  (Default|BindEventtoOnAlertLevelChanged _asbp_player_pawn_tvscare (EventDispatchers|CreateEvent :!event Custom|UpdateAlertLevel)))

(event EventTick
  (params (DeltaSeconds :type Float%20%28single-precision%29)))

(event Custom|UpdateAlertLevel
  (params (OldAlertLevel :type EAlertLevel%20Enum :default "Unaware") (NewAlertLevel :type EAlertLevel%20Enum :default "Unaware"))
  (Variables|Default|SetAlertLevel NewAlertLevel))

(event Custom|AttemptToChangeMusic
  (Utilities|FlowControl|Delay 0.2)
  (switch Utilities|FlowControl|Switch|SwitchonEAlertLevel (Variables|Default|GetAlertLevel)
    (:Unaware
      (Class|NPCMusicManagerComponent|ExecuteMusicTrigger (Variables|Private|GetMusicManager) self (Variables|Default|GetUnaware%20Trigger%20Name)))
    (:Threatened
      (Class|NPCMusicManagerComponent|ExecuteMusicTrigger (Variables|Private|GetMusicManager) self (Variables|Default|GetAggro%20Trigger%20Name)))))

(fn OnRep_AlertLevel
  (Networking|SwitchHasAuthority
    (:Remote
      (CallFunction|AttempttoChangeMusic))))
